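# Prototype hard-surface material, Default Lit, that shades meshes by vertex color channels, with an emissive region on green vertices.
# Red blends base color between black-vert and red-vert parameters.
# Metalness and roughness lerp black to red by the red channel, then toward a green value by the green channel.
# Green also masks an emissive color scaled by a strength parameter, with no textures used.
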Material /Game/Athena/Prototype/HardSurface/Materials/VertexBlend/Vertcolor_Emissive
graph {
 properties {bAutomaticallySetUsageInEditor=false, usedShadingModels=Default Lit}
 nodes [
  n1 = VertexColor()
  LinearInterpolate(
   A=LinearInterpolate(A=ScalarParameter() {parameterName=MetalnessBlackVerts}, B=ScalarParameter() {parameterName=MetalnessRedVerts}, Alpha=@n1.R)
   B=ScalarParameter() {parameterName=MetalessGreenVerts}
   Alpha=@n1.G
  )
  LinearInterpolate(
   A=VectorParameter() {parameterName=Basecolor_BlackVerts, defaultValue=<rgba 0.0445 0.0533 0.099 1>}.RGB
   B=VectorParameter() {parameterName=Basecolor_RedVerts, defaultValue=<rgba 0.227 0.2392 0.5312 1>}.RGB
   Alpha=@n1.R
  )
  LinearInterpolate(
   A=LinearInterpolate(
    A=ScalarParameter() {defaultValue=0.5, parameterName=RoughnessBlackVerts}
    B=ScalarParameter() {defaultValue=0.5, parameterName=RoughnessRedVerts}
    Alpha=@n1.R
   )
   B=ScalarParameter() {defaultValue=0.5, parameterName=RoughnessGreenVerts}
   Alpha=@n1.G
  )
  Multiply(
   A=Multiply(A=VectorParameter() {parameterName=Emissive_GreenVerts, defaultValue=<rgba 0.27 0.2036 0.5312 1>}.RGB, B=@n1.G)
   B=ScalarParameter() {defaultValue=1, parameterName=EmissiveStrength}
  )
 ]
}
